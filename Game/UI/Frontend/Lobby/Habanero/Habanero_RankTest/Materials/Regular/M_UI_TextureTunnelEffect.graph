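# Translucent UI material in the Habanero rank test lobby that builds a tunnel effect from layered copies of one texture.
# It samples T_UI_Habanero_Unreal four times, scaling UVs about the center by a different constant per layer to draw nested sizes.
# Named reroutes carry each layer's color and alpha into chained lerps for color and a clamped alpha sum for opacity.
# The scales are fixed constants, with no animation.
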Material /Game/UI/Frontend/Lobby/Habanero/Habanero_RankTest/Materials/Regular/M_UI_TextureTunnelEffect
graph {
 properties {bAutomaticallySetUsageInEditor=false, blendMode=BLEND_Translucent, materialDomain=MD_UI, usedShadingModels=Default Lit}
 nodes [
  n1 = NamedRerouteUsage()
  n2 = NamedRerouteUsage()
  n3 = NamedRerouteUsage()
  n4 = NamedRerouteUsage()
  Clamp(None=Add(A=Add(A=@n1, B=@n2), B=Add(A=@n3, B=@n4))) {desc=making sure the result does not go above 1}
  n5 = NamedRerouteUsage()
  n6 = NamedRerouteUsage()
  n7 = NamedRerouteUsage()
  LinearInterpolate(
   A=LinearInterpolate(A=LinearInterpolate(A=@n5, B=@n6, Alpha=@n7), B=NamedRerouteUsage(), Alpha=NamedRerouteUsage())
   B=NamedRerouteUsage()
   Alpha=NamedRerouteUsage()
  )
  n8 = Subtract(A=TextureCoordinate(), B=Constant2Vector() {g=0.5, r=0.5})
  n9 = Constant2Vector() {g=0.4, r=0.4}
  n10 = TextureSample(UVs=Add(A=Multiply(A=@n8, B=@n9), B=Constant2Vector() {g=0.5, r=0.5})) {texture=/Game/UI/Frontend/Lobby/Habanero/Habanero_RankTest/Textures/Regular/T_UI_Habanero_Unreal}
  NamedRerouteDeclaration(None=@n10.RGB) {name=Layer01_RGB, nodeColor=<rgba 1 0.8906 0 1>}
  NamedRerouteDeclaration(None=@n10.A) {name=Layer01_A, nodeColor=<rgba 0.5703 0 1 1>}
  n11 = Subtract(A=TextureCoordinate(), B=Constant2Vector() {g=0.5, r=0.5})
  n12 = Constant2Vector() {g=0.5, r=0.5}
  n13 = TextureSample(UVs=Add(A=Multiply(A=@n11, B=@n12), B=Constant2Vector() {g=0.5, r=0.5})) {texture=&T_UI_Habanero_Unreal}
  NamedRerouteDeclaration(None=@n13.A) {name=Layer02_A, nodeColor=<rgba 0 1 0.25 1>}
  NamedRerouteDeclaration(None=@n13.RGB) {name=Layer02_RGB, nodeColor=<rgba 1 0 0.0703 1>}
  n14 = Subtract(A=TextureCoordinate(), B=Constant2Vector() {g=0.5, r=0.5})
  n15 = Constant2Vector() {g=0.75, r=0.75}
  n16 = TextureSample(UVs=Add(A=Multiply(A=@n14, B=@n15), B=Constant2Vector() {g=0.5, r=0.5})) {texture=&T_UI_Habanero_Unreal}
  NamedRerouteDeclaration(None=@n16.RGB) {name=Layer03_RGB, nodeColor=<rgba 0 0.3906 1 1>}
  NamedRerouteDeclaration(None=@n16.A) {name=Layer03_A, nodeColor=<rgba 0.7109 1 0 1>}
  n17 = Subtract(A=TextureCoordinate(), B=Constant2Vector() {g=0.5, r=0.5})
  n18 = Constant2Vector() {g=1, r=1}
  n19 = TextureSample(UVs=Add(A=Multiply(A=@n17, B=@n18), B=Constant2Vector() {g=0.5, r=0.5})) {texture=&T_UI_Habanero_Unreal}
  NamedRerouteDeclaration(None=@n19.RGB) {name=Layer04_RGB, nodeColor=<rgba 0.9688 0 1 1>}
  NamedRerouteDeclaration(None=@n19.A) {name=Layer04_A, nodeColor=<rgba 0 1 0.6484 1>}
 ]
}
